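# Megascans preset master material that turns a single imperfection texture into base color and roughness on a Default Lit surface.
# UVs are scaled by Tiling and offset by the Crop parameters, then sample the texture's red channel.
# That channel drives a blend between Color Intensity and tinted color, passed through SpeedTreeColorVariation for variation, and is scaled for roughness.
# Metallic is a standalone scalar, and all three results carry ms_ labels.
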Material /Bridge/MSPresets/M_MS_Imperfection_Material_Var/M_MS_Imperfection_Material_Var
graph {
 properties {pixelDepthOffsetMode=PDOM_Legacy, usedShadingModels=Default Lit}
 nodes [
  n1 = VectorParameter() {group=Material Settings, parameterName=Base Color, defaultValue=<rgba 0.5 0.5 0.5 1>}
  n2 = TextureCoordinate()
  n3 = ScalarParameter() {defaultValue=1, group=Material Settings, parameterName=Tiling}
  n4 = Constant() {r=1}
  n5 = MaterialFunctionCall(
   X=ScalarParameter() {defaultValue=1, group=Material Settings, parameterName=Crop X}
   Y=ScalarParameter() {defaultValue=1, group=Material Settings, parameterName=Crop Y}
  ) {materialFunction=&MakeFloat2}
  n6 = TextureSampleParameter2D(UVs=Add(A=Multiply(A=@n2, B=@n3), B=Add(A=@n4, B=@n5.Result))) {group=Input Texture, parameterName=Roughness, samplerSource=SSM_Wrap_WorldGroupSettings, texture=/Engine/EngineMaterials/DefaultDiffuse}
  MaterialFunctionCall!(
   BaseColor=LinearInterpolate(A=ScalarParameter() {defaultValue=1, group=Material Settings, parameterName=Color Intensity}, B=Add(A=@n1.RGB, B=@n6.R), Alpha=@n6.R)
   Amount=ScalarParameter() {group=Material Settings, parameterName=Color Variation}
  ) {desc=ms_BaseColor, materialFunction=/Engine/Functions/Engine_MaterialFunctions01/SpeedTree/SpeedTreeColorVariation}
  Multiply(A=@n6.R, B=ScalarParameter() {defaultValue=1, group=Material Settings, parameterName=Roughness Intensity}) {desc=ms_Roughness}
  ScalarParameter() {defaultValue=1, desc=ms_Metallic, group=Material Settings, parameterName=Metallic}
 ]
}
